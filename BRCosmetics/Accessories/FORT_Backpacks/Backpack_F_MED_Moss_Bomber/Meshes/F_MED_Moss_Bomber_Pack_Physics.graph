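# Physics asset for the Moss Bomber backpack cosmetic, giving its dangling rope, vines, leaves, flower petals, handle and zipper secondary motion.
# Each dynamic joint has a capsule body, and a larger capsule on the pack bone anchors them.
# Most parts constrain directly to the pack, while the second rope joint chains to the first and four leaves hang off a master leaf joint.
# Twist is locked throughout, with limited swing and some single-axis hinges.
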PhysicsAsset /BRCosmetics/Accessories/FORT_Backpacks/Backpack_F_MED_Moss_Bomber/Meshes/F_MED_Moss_Bomber_Pack_Physics
bodies {
 dyn_droopingRope_jnt_1 [{shapeType=Capsule, length=3, radius=1, center=<-2.3004 0 0.296>, rotation=<rot 82.6686 0 0>}]
 dyn_droopingRope_jnt_2 [{shapeType=Capsule, length=2, radius=1, center=<-1.7183 0 0.3533>, rotation=<rot 82.6686 0 0>}]
 dyn_flower_petals_jnt_1 [{shapeType=Capsule, length=2, radius=0.3, center=<0.3 0 0>, rotation=<rot 90 0 0>}]
 dyn_flower_petals_jnt_2 [{shapeType=Capsule, length=2, radius=0.3, center=<0.3 0 0>, rotation=<rot 90 0 0>}]
 dyn_flower_petals_jnt_3 [{shapeType=Capsule, length=2, radius=0.3, center=<0.3 0 0>, rotation=<rot 90 0 0>}]
 dyn_handle_jnt [{shapeType=Capsule, length=4, radius=2, center=<3.0177 -1.6742 1.1597>}]
 dyn_hangingVine_jnt_1 [{shapeType=Capsule, length=3, radius=1, center=<2.7027 0 0>, rotation=<rot -90 0 0>}]
 dyn_hangingVine_jnt_2 [{shapeType=Capsule, length=6, radius=1, center=<3.2289 0 0>, rotation=<rot 90 0 0>}]
 dyn_hangingVine_jnt_3 [{shapeType=Capsule, length=3, radius=1, center=<1.8273 -0.0001 0>, rotation=<rot 90 0 0>}]
 dyn_hangingVine_jnt_4 [{shapeType=Capsule, length=1, radius=1, center=<0.8615 0 0>, rotation=<rot 90 0 0>}]
 dyn_hangingVine_jnt_5 [{shapeType=Capsule, length=1, radius=1, center=<1.0083 0 0>, rotation=<rot 90 0 0>}]
 dyn_hangingVine_jnt_6 [{shapeType=Capsule, length=5, radius=1, center=<2.6168 0 0>, rotation=<rot -90 0 0>}]
 dyn_hangingVine_jnt_7 [{shapeType=Capsule, length=5, radius=1, center=<2.0153 0 0>, rotation=<rot 90 0 0>}]
 dyn_leaf_1_jnt [{shapeType=Capsule, length=4, radius=0.4, center=<0.8187 0 -0.0518>, rotation=<rot -86.383 -179.9989 179.9989>}]
 dyn_leaf_2_jnt [{shapeType=Capsule, length=4, radius=0.4, center=<0.8761 0 -0.0554>, rotation=<rot -86.3928 -179.9996 179.9995>}]
 dyn_leaf_3_jnt [{shapeType=Capsule, length=4, radius=0.4, center=<0.4328 0 0.0563>, rotation=<rot -82.5935 -0.0003 0.0003>}]
 dyn_leaf_4_jnt_2 [{shapeType=Capsule, length=4, radius=0.4, center=<0.4328 0 0.0562>, rotation=<rot -82.5995 -0.0004 0.0004>}]
 dyn_leavesMaster_jnt [{shapeType=Capsule, length=3, radius=1, rotation=<rot -74.071 0 0>}]
 dyn_zipper_jnt [{shapeType=Capsule, length=6, radius=0.6, center=<2.6394 0 0.0292>, rotation=<rot 89.4158 179.9998 179.9998>}]
 pack [{shapeType=Capsule, length=8, radius=9, center=<-0.0002 -4.2305 -6.7036>}]
}
constraints [
 {
  bone1Name=dyn_handle_jnt
  bone2Name=pack
  swing1LimitDegrees=20
  swing2LimitDegrees=45
  swing2Motion=Locked
  twistLimitDegrees=45
  twistMotion=Locked
 }
 {
  bone1Name=dyn_leavesMaster_jnt
  bone2Name=pack
  swing1LimitDegrees=1
  swing1Motion=Locked
  swing2LimitDegrees=15
  twistLimitDegrees=1
  twistMotion=Locked
 }
 {
  bone1Name=dyn_leaf_1_jnt
  bone2Name=dyn_leavesMaster_jnt
  swing1LimitDegrees=6
  swing2LimitDegrees=35
  twistLimitDegrees=45
  twistMotion=Locked
 }
 {bone1Name=dyn_zipper_jnt, bone2Name=pack, swing1LimitDegrees=15, swing2LimitDegrees=15, twistLimitDegrees=45, twistMotion=Locked}
 {bone1Name=dyn_droopingRope_jnt_1, bone2Name=pack, swing1LimitDegrees=30, swing2LimitDegrees=30, twistLimitDegrees=45, twistMotion=Locked}
 {
  bone1Name=dyn_droopingRope_jnt_2
  bone2Name=dyn_droopingRope_jnt_1
  swing1LimitDegrees=30
  swing2LimitDegrees=30
  twistLimitDegrees=45
  twistMotion=Locked
 }
 {bone1Name=dyn_hangingVine_jnt_1, bone2Name=pack, swing1LimitDegrees=10, swing2LimitDegrees=10, twistLimitDegrees=45, twistMotion=Locked}
 {bone1Name=dyn_hangingVine_jnt_2, bone2Name=pack, swing1LimitDegrees=20, swing2LimitDegrees=20, twistLimitDegrees=45, twistMotion=Locked}
 {bone1Name=dyn_hangingVine_jnt_3, bone2Name=pack, swing1LimitDegrees=20, swing2LimitDegrees=20, twistLimitDegrees=45, twistMotion=Locked}
 {bone1Name=dyn_hangingVine_jnt_4, bone2Name=pack, swing1LimitDegrees=20, swing2LimitDegrees=20, twistLimitDegrees=45, twistMotion=Locked}
 {bone1Name=dyn_hangingVine_jnt_5, bone2Name=pack, swing1LimitDegrees=20, swing2LimitDegrees=20, twistLimitDegrees=45, twistMotion=Locked}
 {bone1Name=dyn_hangingVine_jnt_6, bone2Name=pack, swing1LimitDegrees=30, swing2LimitDegrees=30, twistLimitDegrees=45, twistMotion=Locked}
 {bone1Name=dyn_hangingVine_jnt_7, bone2Name=pack, swing1LimitDegrees=20, swing2LimitDegrees=20, twistLimitDegrees=45, twistMotion=Locked}
 {
  bone1Name=dyn_leaf_2_jnt
  bone2Name=dyn_leavesMaster_jnt
  swing1LimitDegrees=6
  swing2LimitDegrees=40
  twistLimitDegrees=45
  twistMotion=Locked
 }
 {
  bone1Name=dyn_leaf_4_jnt_2
  bone2Name=dyn_leavesMaster_jnt
  swing1LimitDegrees=6
  swing2LimitDegrees=35
  twistLimitDegrees=45
  twistMotion=Locked
 }
 {
  bone1Name=dyn_leaf_3_jnt
  bone2Name=dyn_leavesMaster_jnt
  swing1LimitDegrees=6
  swing2LimitDegrees=35
  twistLimitDegrees=45
  twistMotion=Locked
 }
 {
  bone1Name=dyn_flower_petals_jnt_1
  bone2Name=pack
  swing1LimitDegrees=45
  swing1Motion=Locked
  swing2LimitDegrees=20
  twistLimitDegrees=45
  twistMotion=Locked
 }
 {
  bone1Name=dyn_flower_petals_jnt_2
  bone2Name=pack
  swing1LimitDegrees=45
  swing1Motion=Locked
  swing2LimitDegrees=20
  twistLimitDegrees=45
  twistMotion=Locked
 }
 {
  bone1Name=dyn_flower_petals_jnt_3
  bone2Name=pack
  swing1LimitDegrees=45
  swing1Motion=Locked
  swing2LimitDegrees=20
  twistLimitDegrees=45
  twistMotion=Locked
 }
]
